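# Sound cue for the DPS shotgun's aim-down-sights exit foley in Flipper gameplay.
# A single ADS-out wave plays through a modulator that randomizes pitch slightly upward for variation.
# It plays quietly as weapon foley, using the shared weapon reload attenuation for spatial falloff.
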SoundCue /FlipperGameplay/Items/Weapons/DPSShotgun/Audio/SC_Shotgun_CoreDPS_ADS_OUT
root=Modulator_0
nodes [
 {id=Modulator_0, children [WavePlayer_0], properties {pitchMax=1.2, pitchMin=1}}
 {id=WavePlayer_0, properties {soundWaveAssetPtr=/FlipperGameplay/Items/Weapons/DPSShotgun/Audio/SoundWaves/DPS_Shotgun_ADS_Out_01}}
]
properties {
 attenuationSettings=&Fort_Weapon_Reload_Att
 maxDistance=2250
 soundClassObject=&Fort_WeaponFoley
 volumeMultiplier=0.1
}
