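# Sound cue for the destroy sound of the CoreDPS shotgun, part of the FlipperGameplay DPS shotgun weapon audio.
# A single wave player feeds the output with the shared Fort_Weapons shotgun_destroy_01 wave, with no randomization or modulation.
# It routes to the ranged weapon low-ammo/degraded sound class and overrides concurrency so new instances are rejected once the limit is reached.
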SoundCue /FlipperGameplay/Items/Weapons/DPSShotgun/Audio/SC_Shotgun_CoreDPS_Destroy
root=WavePlayer_65
nodes [{id=WavePlayer_65, properties {soundWaveAssetPtr=/Game/Sounds/Fort_Weapons/pistol/degrades/shotgun_destroy_01}}]
properties {
 bOverrideConcurrency=true
 maxDistance=2097152
 soundClassObject=/Game/Sounds/FortSoundClassesAndModes/Fort_WeaponRanged_LowAmmo_Degraded_PSM
 volumeMultiplier=1
 concurrencyOverrides {resolutionRule=PreventNew}
}
