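# Sound cue for the DPS shotgun's drop impact sound, routed to the Fort_Loot_Collect sound class.
# A single wave player feeds a default Modulator at the root, with the overall volume reduced.
# Inline attenuation and concurrency overrides give it a short-range falloff and a small instance cap that blocks new plays once full.
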SoundCue /FlipperGameplay/Items/Weapons/DPSShotgun/Audio/SC_Shotgun_CoreDPS_DropImpact
root=Modulator_0
nodes [
 {id=Modulator_0, children [WavePlayer_2]}
 {id=WavePlayer_2, properties {soundWaveAssetPtr=/FlipperGameplay/Items/Weapons/DPSShotgun/Audio/SoundWaves/DPS_Shotgun_Drop_Impact}}
]
properties {
 bOverrideAttenuation=true
 bOverrideConcurrency=true
 maxDistance=1000
 soundClassObject=&Fort_Loot_Collect
 volumeMultiplier=0.4
 attenuationOverrides {falloffDistance=1000, attenuationShapeExtents {x=0}}
 concurrencyOverrides {maxCount=3, resolutionRule=PreventNew}
}
