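# Sound cue for the equip sound of the CoreDPS shotgun weapon in Flipper gameplay.
# A single wave player plays the DPS shotgun equip wave at reduced volume, with no randomization or mixing.
# It routes through the weapon foley sound class and reload attenuation, with its own owner-limited concurrency that stops the oldest instance.
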SoundCue /FlipperGameplay/Items/Weapons/DPSShotgun/Audio/SC_Shotgun_CoreDPS_Equip
root=WavePlayer_0
nodes [{id=WavePlayer_0, properties {soundWaveAssetPtr=/FlipperGameplay/Items/Weapons/DPSShotgun/Audio/SoundWaves/DPS_Shotgun_Equip}}]
properties {
 attenuationSettings=&Fort_Weapon_Reload_Att
 bOverrideConcurrency=true
 maxDistance=2250
 soundClassObject=&Fort_WeaponFoley
 volumeMultiplier=0.3
 attenuationOverrides {falloffDistance=1600}
 concurrencyOverrides {bLimitToOwner=true, maxCount=2, resolutionRule=StopOldest}
}
