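# Sound cue for the DPS shotgun's core firing sound, layered by distance with concurrency limits.
# Close, medium and distant layers randomize wave variations, each with its own attenuation and modulation, under third-person fire or distant sound classes.
# The first quality slot adds a close thump beside the briefly delayed shot and uses fuller variation pools.
# The other three slots share a lighter mix with two variations per layer and no thump.
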SoundCue /FlipperGameplay/Items/Weapons/DPSShotgun/Audio/SC_Shotgun_CoreDPS_Fire
root=QualityLevel_0
nodes [
 {id=Attenuation_0, children [Modulator_4], properties {attenuationSettings=&Fort_Weapon_Projectile_Close_Athena_Att}}
 {
  id=Attenuation_1
  children [Modulator_1]
  properties {attenuationSettings=&Fort_Weapon_Projectile_Medium_Mono_Athena_Att}
 }
 {id=Attenuation_2, children [Mixer_1], properties {attenuationSettings=&Fort_Weapon_Projectile_Close_Athena_Att}}
 {
  id=Attenuation_3
  children [Modulator_6]
  properties {attenuationSettings=&Fort_Weapon_Projectile_Distant_Mono_Athena_Att}
 }
 {
  id=Attenuation_7
  children [Modulator_2]
  properties {attenuationSettings=&Fort_Weapon_Projectile_Distant_Mono_Athena_Att}
 }
 {
  id=Attenuation_8
  children [Modulator_5]
  properties {attenuationSettings=&Fort_Weapon_Projectile_Medium_Mono_Athena_Att}
 }
 {id=Delay_0, children [Random_3], properties {delayMax=0.0011, delayMin=0.001}}
 {id=Mixer_0, children [SoundClass_0,SoundClass_8,SoundClass_7], properties {inputVolume [0.8,0.45,0.55]}}
 {id=Mixer_1, children [Modulator_0,Modulator_3], properties {inputVolume [1,1]}}
 {id=Mixer_3, children [SoundClass_2,SoundClass_6,SoundClass_3], properties {inputVolume [0.8,0.45,0.55]}}
 {id=Modulator_0, children [Delay_0], properties {pitchMin=0.975}}
 {id=Modulator_1, children [Random_0], properties {volumeMax=2, volumeMin=2}}
 {id=Modulator_2, children [Random_7], properties {pitchMax=1, pitchMin=0.9, volumeMax=2, volumeMin=2}}
 {id=Modulator_3, children [Random_4], properties {pitchMin=0.975}}
 {id=Modulator_4, children [Random_5], properties {volumeMax=1.1, volumeMin=1}}
 {id=Modulator_5, children [Random_6], properties {volumeMax=2, volumeMin=2}}
 {id=Modulator_6, children [Random_2], properties {pitchMax=1, pitchMin=0.9, volumeMax=2, volumeMin=2}}
 {id=QualityLevel_0, children [Mixer_3,Mixer_0,Mixer_0,Mixer_0]}
 {id=Random_0, children [WavePlayer_0,WavePlayer_1,WavePlayer_2], properties {weights [1,1,1]}}
 {id=Random_2, children [WavePlayer_11, WavePlayer_12, WavePlayer_13], properties {weights [1,1,1]}}
 {id=Random_3, children [WavePlayer_5,WavePlayer_6,WavePlayer_7], properties {weights [1,1,1]}}
 {id=Random_4, children [WavePlayer_8,WavePlayer_9], properties {weights [1,1]}}
 {id=Random_5, children [WavePlayer_14, WavePlayer_15], properties {weights [1,1]}}
 {id=Random_6, children [WavePlayer_16, WavePlayer_17], properties {weights [1,1]}}
 {id=Random_7, children [WavePlayer_18, WavePlayer_26], properties {weights [1,1]}}
 {
  id=SoundClass_0
  children [Attenuation_0]
  properties {soundClassOverride=&Fort_WeaponRanged_Fire_3P_PSM}
 }
 {
  id=SoundClass_2
  children [Attenuation_2]
  properties {soundClassOverride=&Fort_WeaponRanged_Fire_3P_PSM}
 }
 {
  id=SoundClass_3
  children [Attenuation_3]
  properties {soundClassOverride=&Fort_WeaponRanged_Fire_Distant}
 }
 {
  id=SoundClass_6
  children [Attenuation_1]
  properties {soundClassOverride=&Fort_WeaponRanged_Fire_Distant}
 }
 {
  id=SoundClass_7
  children [Attenuation_7]
  properties {soundClassOverride=&Fort_WeaponRanged_Fire_Distant}
 }
 {
  id=SoundClass_8
  children [Attenuation_8]
  properties {soundClassOverride=&Fort_WeaponRanged_Fire_Distant}
 }
 {id=WavePlayer_0, properties {soundWaveAssetPtr=/FlipperGameplay/Items/Weapons/DPSShotgun/Audio/SoundWaves/DPS_Shotgun_Fire_03_Med}}
 {id=WavePlayer_1, properties {soundWaveAssetPtr=/FlipperGameplay/Items/Weapons/DPSShotgun/Audio/SoundWaves/DPS_Shotgun_Fire_02_Med}}
 {id=WavePlayer_11, properties {soundWaveAssetPtr=/FlipperGameplay/Items/Weapons/DPSShotgun/Audio/SoundWaves/DPS_Shotgun_Fire_01_Dstnt}}
 {id=WavePlayer_12, properties {soundWaveAssetPtr=/FlipperGameplay/Items/Weapons/DPSShotgun/Audio/SoundWaves/DPS_Shotgun_Fire_02_Dstnt}}
 {id=WavePlayer_13, properties {soundWaveAssetPtr=/FlipperGameplay/Items/Weapons/DPSShotgun/Audio/SoundWaves/DPS_Shotgun_Fire_03_Dstnt}}
 {id=WavePlayer_14, properties {soundWaveAssetPtr=/FlipperGameplay/Items/Weapons/DPSShotgun/Audio/SoundWaves/DPS_Shotgun_Fire_01}}
 {id=WavePlayer_15, properties {soundWaveAssetPtr=/FlipperGameplay/Items/Weapons/DPSShotgun/Audio/SoundWaves/DPS_Shotgun_Fire_02}}
 {id=WavePlayer_16, properties {soundWaveAssetPtr=&DPS_Shotgun_Fire_03_Med}}
 {id=WavePlayer_17, properties {soundWaveAssetPtr=&DPS_Shotgun_Fire_02_Med}}
 {id=WavePlayer_18, properties {soundWaveAssetPtr=&DPS_Shotgun_Fire_01_Dstnt}}
 {id=WavePlayer_2, properties {soundWaveAssetPtr=/FlipperGameplay/Items/Weapons/DPSShotgun/Audio/SoundWaves/DPS_Shotgun_Fire_01_Med}}
 {id=WavePlayer_26, properties {soundWaveAssetPtr=&DPS_Shotgun_Fire_02_Dstnt}}
 {id=WavePlayer_5, properties {soundWaveAssetPtr=&DPS_Shotgun_Fire_01}}
 {id=WavePlayer_6, properties {soundWaveAssetPtr=&DPS_Shotgun_Fire_02}}
 {id=WavePlayer_7, properties {soundWaveAssetPtr=/FlipperGameplay/Items/Weapons/DPSShotgun/Audio/SoundWaves/DPS_Shotgun_Fire_03}}
 {id=WavePlayer_8, properties {soundWaveAssetPtr=/FlipperGameplay/Items/Weapons/DPSShotgun/Audio/SoundWaves/DPS_Shotgun_Thump_01}}
 {id=WavePlayer_9, properties {soundWaveAssetPtr=/FlipperGameplay/Items/Weapons/DPSShotgun/Audio/SoundWaves/DPS_Shotgun_Thump_02}}
]
properties {
 maxDistance=27500
 attenuationOverrides {falloffDistance=5000, attenuationShapeExtents {x=1000}}
 concurrencyOverrides {bLimitToOwner=true, maxCount=3}
 concurrencySet [&Weapon_Gunfire_Local_Conc, &Weapon_Gunfire_Global_Conc]
}
