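# First-person fire sound cue for the DPS Shotgun weapon.
# A mixer layers a randomly chosen fire wave from three variants, offset by a tiny delay, with a quieter randomly chosen thump from two.
# Each layer passes through a pitch modulator for slight variation.
# It uses first-person weapon attenuation, a ranged-fire sound class and a shared first-person gunfire concurrency set.
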SoundCue /FlipperGameplay/Items/Weapons/DPSShotgun/Audio/SC_Shotgun_CoreDPS_Fire_1P
root=Mixer_0
nodes [
 {id=Delay_0, children [Random_0], properties {delayMax=0.0011, delayMin=0.001}}
 {id=Mixer_0, children [Modulator_0,Modulator_1], properties {inputVolume [1,0.25]}}
 {id=Modulator_0, children [Delay_0], properties {pitchMin=0.975}}
 {id=Modulator_1, children [Random_2], properties {pitchMin=0.975}}
 {id=Random_0, children [WavePlayer_0,WavePlayer_1,WavePlayer_6], properties {weights [1,1,1]}}
 {id=Random_2, children [WavePlayer_7,WavePlayer_8], properties {weights [1,1]}}
 {id=WavePlayer_0, properties {soundWaveAssetPtr=/FlipperGameplay/Items/Weapons/DPSShotgun/Audio/SoundWaves/DPS_Shotgun_Fire_01}}
 {id=WavePlayer_1, properties {soundWaveAssetPtr=/FlipperGameplay/Items/Weapons/DPSShotgun/Audio/SoundWaves/DPS_Shotgun_Fire_02}}
 {id=WavePlayer_6, properties {soundWaveAssetPtr=/FlipperGameplay/Items/Weapons/DPSShotgun/Audio/SoundWaves/DPS_Shotgun_Fire_03}}
 {id=WavePlayer_7, properties {soundWaveAssetPtr=/FlipperGameplay/Items/Weapons/DPSShotgun/Audio/SoundWaves/DPS_Shotgun_Thump_01}}
 {id=WavePlayer_8, properties {soundWaveAssetPtr=/FlipperGameplay/Items/Weapons/DPSShotgun/Audio/SoundWaves/DPS_Shotgun_Thump_02}}
]
properties {
 attenuationSettings=&Fort_Weapon_Projectile_FirstPerson_Athena_Att
 maxDistance=2097152
 soundClassObject=&Fort_WeaponRanged_Fire_1P_PSM
 volumeMultiplier=0.9
 attenuationOverrides {falloffDistance=5000, attenuationShapeExtents {x=1000}}
 concurrencyOverrides {bLimitToOwner=true, maxCount=3, resolutionRule=StopOldest}
 concurrencySet [&Weapon_Gunfire_FirstPerson_Conc]
}
