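# Sound cue for the low-ammo sound of the CoreDPS shotgun, a Flipper gameplay weapon.
# A single low-ammo wave player feeds a Modulator at the root, which is left at its default settings.
# The cue routes to the degraded ranged-weapon low-ammo sound class and overrides concurrency with a PreventNew rule, so new instances are refused once the limit is reached.
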SoundCue /FlipperGameplay/Items/Weapons/DPSShotgun/Audio/SC_Shotgun_CoreDPS_LowAmmo
root=Modulator_0
nodes [
 {id=Modulator_0, children [WavePlayer_0]}
 {id=WavePlayer_0, properties {soundWaveAssetPtr=/FlipperGameplay/Items/Weapons/DPSShotgun/Audio/SoundWaves/DPS_Shotgun_LowAmmo_01}}
]
properties {
 bOverrideConcurrency=true
 maxDistance=2097152
 soundClassObject=/Game/Sounds/FortSoundClassesAndModes/Fort_WeaponRanged_LowAmmo_Degraded_PSM
 concurrencyOverrides {resolutionRule=PreventNew}
}
