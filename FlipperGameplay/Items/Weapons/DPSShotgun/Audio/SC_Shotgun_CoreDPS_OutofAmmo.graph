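# Sound cue for the dry-fire sound of the DPS shotgun when it is out of ammo.
# A Random node picks one of three equally weighted dry-fire waves, fed through a Modulator that adds slight volume variation with fixed pitch.
# The cue plays at reduced volume under a ranged-weapon fire sound class, with overridden concurrency that stops the oldest instance beyond a small cap.
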SoundCue /FlipperGameplay/Items/Weapons/DPSShotgun/Audio/SC_Shotgun_CoreDPS_OutofAmmo
root=Modulator_0
nodes [
 {id=Modulator_0, children [Random_1], properties {pitchMax=1, pitchMin=1, volumeMax=1.1, volumeMin=1}}
 {id=Random_1, children [WavePlayer_0,WavePlayer_5,WavePlayer_6], properties {weights [1,1,1]}}
 {id=WavePlayer_0, properties {soundWaveAssetPtr=/FlipperGameplay/Items/Weapons/DPSShotgun/Audio/SoundWaves/DPS_Shotgun_DryFire_01}}
 {id=WavePlayer_5, properties {soundWaveAssetPtr=/FlipperGameplay/Items/Weapons/DPSShotgun/Audio/SoundWaves/DPS_Shotgun_DryFire_02}}
 {id=WavePlayer_6, properties {soundWaveAssetPtr=/FlipperGameplay/Items/Weapons/DPSShotgun/Audio/SoundWaves/DPS_Shotgun_DryFire_03}}
]
properties {
 bOverrideConcurrency=true
 maxDistance=2097152
 soundClassObject=&Fort_WeaponRanged_Fire_3P_PSM
 volumeMultiplier=0.5
 attenuationOverrides {falloffDistance=1600}
 concurrencyOverrides {maxCount=2, resolutionRule=StopOldest}
}
