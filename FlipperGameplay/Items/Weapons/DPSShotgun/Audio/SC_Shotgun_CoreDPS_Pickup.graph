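# Sound cue for picking up the DPS shotgun, layering a main pickup sound with a delayed accent.
# A mixer blends a random pick of three pickup waves, pitched slightly up, with a quieter, briefly delayed random pick of two accent waves.
# It uses weapon-pickup attenuation and the loot-collect sound class at reduced volume.
# A concurrency override allows only one instance at a time, preventing new plays while one sounds.
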SoundCue /FlipperGameplay/Items/Weapons/DPSShotgun/Audio/SC_Shotgun_CoreDPS_Pickup
root=Mixer_1
nodes [
 {id=Delay_0, children [Random_2], properties {delayMax=0.1, delayMin=0.1}}
 {id=Mixer_1, children [Delay_0,Modulator_0], properties {inputVolume [0.5,1]}}
 {id=Modulator_0, children [Random_1], properties {pitchMax=1.2, pitchMin=1.1}}
 {id=Random_1, children [WavePlayer_2,WavePlayer_3,WavePlayer_4], properties {weights [1,1,1]}}
 {id=Random_2, children [WavePlayer_5,WavePlayer_6], properties {weights [1,1]}}
 {id=WavePlayer_2, properties {soundWaveAssetPtr=/FlipperGameplay/Items/Weapons/DPSShotgun/Audio/SoundWaves/DPS_Shotgun_PickUp_01}}
 {id=WavePlayer_3, properties {soundWaveAssetPtr=/FlipperGameplay/Items/Weapons/DPSShotgun/Audio/SoundWaves/DPS_Shotgun_PickUp_02}}
 {id=WavePlayer_4, properties {soundWaveAssetPtr=/FlipperGameplay/Items/Weapons/DPSShotgun/Audio/SoundWaves/DPS_Shotgun_Pickup_03}}
 {id=WavePlayer_5, properties {soundWaveAssetPtr=/FlipperGameplay/Items/Weapons/DPSShotgun/Audio/SoundWaves/DPS_Shotgun_PickUp_Acc_01}}
 {id=WavePlayer_6, properties {soundWaveAssetPtr=/FlipperGameplay/Items/Weapons/DPSShotgun/Audio/SoundWaves/DPS_Shotgun_PickUp_Acc_02}}
]
properties {
 attenuationSettings=&Fort_Weapon_Pickup_Att
 bOverrideConcurrency=true
 maxDistance=3500
 soundClassObject=&Fort_Loot_Collect
 volumeMultiplier=0.25
 concurrencyOverrides {maxCount=1, resolutionRule=PreventNew}
}
