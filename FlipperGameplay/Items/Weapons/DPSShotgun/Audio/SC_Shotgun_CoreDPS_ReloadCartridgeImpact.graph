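# Sound cue for the DPS shotgun's reload cartridge impact, playing a shell bounce sound.
# A Modulator root varies the output of an equally weighted Random node that picks one of three shell bounce waves.
# It plays quietly in the weapon foley sound class with reload attenuation.
# Its own concurrency override rejects new instances once the limit is reached, rather than stopping playing ones.
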SoundCue /FlipperGameplay/Items/Weapons/DPSShotgun/Audio/SC_Shotgun_CoreDPS_ReloadCartridgeImpact
root=Modulator_0
nodes [
 {id=Modulator_0, children [Random_4]}
 {id=Random_4, children [WavePlayer_10, WavePlayer_11, WavePlayer_12], properties {weights [1,1,1]}}
 {id=WavePlayer_10, properties {soundWaveAssetPtr=/FlipperGameplay/Items/Weapons/DPSShotgun/Audio/SoundWaves/DPS_Shotgun_Shell_Bounce_01}}
 {id=WavePlayer_11, properties {soundWaveAssetPtr=/FlipperGameplay/Items/Weapons/DPSShotgun/Audio/SoundWaves/DPS_Shotgun_Shell_Bounce_02}}
 {id=WavePlayer_12, properties {soundWaveAssetPtr=/FlipperGameplay/Items/Weapons/DPSShotgun/Audio/SoundWaves/DPS_Shotgun_Shell_Bounce_03}}
]
properties {
 attenuationSettings=&Fort_Weapon_Reload_Att
 bOverrideConcurrency=true
 maxDistance=2250
 soundClassObject=&Fort_WeaponFoley
 volumeMultiplier=0.1
 attenuationOverrides {falloffDistance=1600}
 concurrencyOverrides {resolutionRule=PreventNew}
}
