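# Sound cue for the end of the DPS Shotgun's reload, played as weapon foley.
# A Random node picks one of three equally weighted ReloadEnd wave variations and feeds it through a Modulator to the output.
# It plays at reduced volume with shared weapon-reload attenuation.
# A concurrency override limits instances per owner and stops the oldest when the cap is exceeded.
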SoundCue /FlipperGameplay/Items/Weapons/DPSShotgun/Audio/SC_Shotgun_CoreDPS_ReloadEnd
root=Modulator_14
nodes [
 {id=Modulator_14, children [Random_0]}
 {id=Random_0, children [WavePlayer_1,WavePlayer_2,WavePlayer_3], properties {weights [1,1,1]}}
 {id=WavePlayer_1, properties {soundWaveAssetPtr=/FlipperGameplay/Items/Weapons/DPSShotgun/Audio/SoundWaves/DPS_Shotgun_ReloadEnd_01}}
 {id=WavePlayer_2, properties {soundWaveAssetPtr=/FlipperGameplay/Items/Weapons/DPSShotgun/Audio/SoundWaves/DPS_Shotgun_ReloadEnd_02}}
 {id=WavePlayer_3, properties {soundWaveAssetPtr=/FlipperGameplay/Items/Weapons/DPSShotgun/Audio/SoundWaves/DPS_Shotgun_ReloadEnd_03}}
]
properties {
 attenuationSettings=&Fort_Weapon_Reload_Att
 bOverrideConcurrency=true
 maxDistance=2250
 soundClassObject=&Fort_WeaponFoley
 volumeMultiplier=0.25
 attenuationOverrides {falloffDistance=1600}
 concurrencyOverrides {bLimitToOwner=true, maxCount=2, resolutionRule=StopOldest}
}
