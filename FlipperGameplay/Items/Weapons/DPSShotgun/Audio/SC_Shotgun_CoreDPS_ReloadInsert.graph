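# Sound cue for the shell-insert foley in the CoreDPS shotgun's reload, routed to the weapon foley sound class with reload attenuation.
# A quality-level switch sends its first two levels to a random pick of five insert waves and its last two to a pick of the first three.
# A root modulator adds slight pitch variation.
# Its concurrency override prevents new instances once the limit is reached.
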SoundCue /FlipperGameplay/Items/Weapons/DPSShotgun/Audio/SC_Shotgun_CoreDPS_ReloadInsert
root=Modulator_0
nodes [
 {id=Modulator_0, children [QualityLevel_1], properties {pitchMax=1.1, pitchMin=0.9}}
 {id=QualityLevel_1, children [Random_2,Random_2,Random_3,Random_3]}
 {id=Random_2, children [WavePlayer_6, WavePlayer_7, WavePlayer_8, WavePlayer_9, WavePlayer_10], properties {weights [1,1,1,1,1]}}
 {id=Random_3, children [WavePlayer_11, WavePlayer_12, WavePlayer_13], properties {weights [1,1,1]}}
 {id=WavePlayer_10, properties {soundWaveAssetPtr=/FlipperGameplay/Items/Weapons/DPSShotgun/Audio/SoundWaves/DPS_Shotgun_Insert_05}}
 {id=WavePlayer_11, properties {soundWaveAssetPtr=/FlipperGameplay/Items/Weapons/DPSShotgun/Audio/SoundWaves/DPS_Shotgun_Insert_01}}
 {id=WavePlayer_12, properties {soundWaveAssetPtr=/FlipperGameplay/Items/Weapons/DPSShotgun/Audio/SoundWaves/DPS_Shotgun_Insert_02}}
 {id=WavePlayer_13, properties {soundWaveAssetPtr=/FlipperGameplay/Items/Weapons/DPSShotgun/Audio/SoundWaves/DPS_Shotgun_Insert_03}}
 {id=WavePlayer_6, properties {soundWaveAssetPtr=&DPS_Shotgun_Insert_01}}
 {id=WavePlayer_7, properties {soundWaveAssetPtr=&DPS_Shotgun_Insert_02}}
 {id=WavePlayer_8, properties {soundWaveAssetPtr=&DPS_Shotgun_Insert_03}}
 {id=WavePlayer_9, properties {soundWaveAssetPtr=/FlipperGameplay/Items/Weapons/DPSShotgun/Audio/SoundWaves/DPS_Shotgun_Insert_04}}
]
properties {
 attenuationSettings=&Fort_Weapon_Reload_Att
 bOverrideConcurrency=true
 maxDistance=2250
 soundClassObject=&Fort_WeaponFoley
 volumeMultiplier=0.25
 attenuationOverrides {falloffDistance=1600}
 concurrencyOverrides {resolutionRule=PreventNew}
}
