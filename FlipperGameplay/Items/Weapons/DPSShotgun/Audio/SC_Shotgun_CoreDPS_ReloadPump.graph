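# Sound cue for the pump action sound played when reloading the DPS shotgun.
# A single pump wave plays through a modulator that slightly randomizes pitch, at a reduced cue volume.
# It uses weapon reload attenuation and the weapon foley sound class.
# A concurrency override limits it to a couple of instances per owner, stopping the oldest first.
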SoundCue /FlipperGameplay/Items/Weapons/DPSShotgun/Audio/SC_Shotgun_CoreDPS_ReloadPump
root=Modulator_14
nodes [
 {id=Modulator_14, children [WavePlayer_1], properties {pitchMax=1.1, pitchMin=0.9}}
 {id=WavePlayer_1, properties {soundWaveAssetPtr=/FlipperGameplay/Items/Weapons/DPSShotgun/Audio/SoundWaves/DPS_Shotgun_Pump}}
]
properties {
 attenuationSettings=&Fort_Weapon_Reload_Att
 bOverrideConcurrency=true
 maxDistance=2250
 soundClassObject=&Fort_WeaponFoley
 volumeMultiplier=0.2
 attenuationOverrides {falloffDistance=1600}
 concurrencyOverrides {bLimitToOwner=true, maxCount=2, resolutionRule=StopOldest}
}
