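# Sound cue for the DPS Shotgun's reload-start sound in the Flipper gameplay content.
# A Random node picks one of three equally weighted reload-start waves and feeds a Modulator at the root.
# It plays quietly in the weapon foley sound class with shared weapon reload attenuation.
# Its own concurrency override blocks new instances once the limit is reached instead of stopping playing ones.
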SoundCue /FlipperGameplay/Items/Weapons/DPSShotgun/Audio/SC_Shotgun_CoreDPS_ReloadStart
root=Modulator_0
nodes [
 {id=Modulator_0, children [Random_3]}
 {id=Random_3, children [WavePlayer_7,WavePlayer_8,WavePlayer_9], properties {weights [1,1,1]}}
 {id=WavePlayer_7, properties {soundWaveAssetPtr=/FlipperGameplay/Items/Weapons/DPSShotgun/Audio/SoundWaves/DPS_Shotgun_Reload_Start_01}}
 {id=WavePlayer_8, properties {soundWaveAssetPtr=/FlipperGameplay/Items/Weapons/DPSShotgun/Audio/SoundWaves/DPS_Shotgun_Reload_Start_02}}
 {id=WavePlayer_9, properties {soundWaveAssetPtr=/FlipperGameplay/Items/Weapons/DPSShotgun/Audio/SoundWaves/DPS_Shotgun_Reload_Start_03}}
]
properties {
 attenuationSettings=&Fort_Weapon_Reload_Att
 bOverrideConcurrency=true
 maxDistance=2250
 soundClassObject=&Fort_WeaponFoley
 volumeMultiplier=0.15
 attenuationOverrides {falloffDistance=1600}
 concurrencyOverrides {resolutionRule=PreventNew}
}
