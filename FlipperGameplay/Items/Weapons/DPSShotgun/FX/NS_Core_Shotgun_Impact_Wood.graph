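# Niagara system for a shotgun impact on wood, built from six one-shot burst emitters with no user parameters.
# Two brief mesh flash emitters and a FlashStreaks emitter give the hot orange hit, using curve-driven color and scale.
# GPU-simulated mesh chunks and dark brown sprite splinters fly out in cones and fall, while an optional smoke puff erodes through dynamic material parameters.
# Streaks, sprite debris and smoke use distance culling.
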NiagaraSystem /FlipperGameplay/Items/Weapons/DPSShotgun/FX/NS_Core_Shotgun_Impact_Wood
dependencies {
 usedDataInterfaces [&NiagaraDataInterfaceVectorCurve, &NiagaraDataInterfaceCurve, &NiagaraDataInterfaceColorCurve]
 usedDynamicInputs []
 usedModules [&SystemState]
}
emitters {
 FN_Weapons_Core_Flash_BETA {
  emitterUpdateScript [
   &EmitterState()
   &SpawnBurst_Instantaneous(Spawn Count=1, Use Spawn Probability=false)
  ]
  particleSpawnScript [
   /Niagara/Modules/Spawn/Initialization/InitializeParticle(
    Lifetime=&UniformRangedFloat(Minimum=0.08, Maximum=0.1)
    Mesh Scale=&UniformRangedVector(Minimum=<0.5 0.3 0.3>, Maximum=<0.8 0.44 0.44>)
    Module.Write Color=true
    Module.Write Lifetime=true
    Module.Write Mass=true
    Module.Write Material Random=true
    Module.Write Position=true
    Module.Write Scale=true
    Module.Write SpriteRotation=false
    Module.Write SpriteSize=true
    Position=&Add_Vector(A=/Niagara/DynamicInputs/Helpers/SimulationPosition(), B=<0 0 0>)
    Sprite Size=<100 100>
   )
   /Game/Effects/Niagara/Modules/Orientation/FN_Fan(Spin:NiagaraFloat=&UniformRangedFloat(Minimum=0, Maximum=1))
   /SetVariables(Particles.LightPositionNew:Vector3f=&Add_Vector(
     A=Linked() {linkedVariable {name=Particles.Position, type=NiagaraPosition}}
     B=/Niagara/DynamicInputs/Transforms/TransformPosition(Position=<25 0 0>, Source Space=::Local, Destination Space=::Simulation)
    )) {isSetParameters=true, module=SetVariables_#1}
   /SetVariables(Particles.SpriteSize=<5 5>) {isSetParameters=true}
   &SubUVAnimation(End Frame=3, SubUV Animation Mode:ENiagara_SubUVLookupMode=::NewEnumerator2)
   &CameraOffset(Camera Offset Amount=10)
   &InitialMeshOrientation(
    Mesh Orientation Mode=::NewEnumerator2
    Rotation=&RandomRangeVector(Minimum=<0 0 0>, Maximum=<1 0 0>)
    Use Rotation Vector=true
   )
  ]
  particleUpdateScript [
   &ParticleState()
   &Color(
    Color=&MakeLinearColorFromVectorAndFloat(
     "Vector (RGB)"=&MakeVectorFromLinearColorRGB(Color=&ColorFromCurve(ColorCurve=DataInterface() {propertyValues {
         AlphaCurve {keys []}
         BlueCurve {keys [
           {arriveTangent=0.0059, interpMode=RCIM_Cubic, leaveTangent=0.0059, time=0, value=0}
           {arriveTangent=0.0059, interpMode=RCIM_Cubic, leaveTangent=0.0059, time=0.9977, value=0.0059}
          ]}
         GreenCurve {keys [
           {arriveTangent=-4.9082, interpMode=RCIM_Cubic, leaveTangent=-4.9082, time=0, value=4.903}
           {arriveTangent=-4.9082, interpMode=RCIM_Cubic, leaveTangent=-4.9082, time=0.9977, value=0.0059}
          ]}
         RedCurve {keys [
           {arriveTangent=-14.4701, interpMode=RCIM_Cubic, leaveTangent=-14.4701, time=0, value=15}
           {arriveTangent=-14.4701, interpMode=RCIM_Cubic, leaveTangent=-14.4701, time=0.9977, value=0.5625}
          ]}
        }}))
     "Float (Alpha)"=&FloatFromCurve(FloatCurve=DataInterface() {propertyValues {Curve {keys [{interpMode=RCIM_Cubic, tangentMode=RCTM_User, time=0, value=1}, {interpMode=RCIM_Cubic, tangentMode=RCTM_User, time=1, value=0}]}}})
    )
    Scale Color=&VectorFromFloat(Value=&FloatFromCurve(FloatCurve=DataInterface() {propertyValues {Curve {keys [{interpMode=RCIM_Cubic, tangentMode=RCTM_User, time=0, value=1}, {interpMode=RCIM_Cubic, tangentMode=RCTM_User, time=1, value=0}]}}}))
   )
   &ScaleMeshSize(Scale Factor=&VectorFromFloat(Value=&FloatFromCurve(FloatCurve=DataInterface() {propertyValues {Curve {keys [{arriveTangent=2.2457, interpMode=RCIM_Cubic, leaveTangent=2.2457, tangentMode=RCTM_User, time=0, value=0}, <at 1 1>]}}})))
   &FN_Sprite_Attributes(
    Create SpriteSize Attribute=true
    SpriteSize=&Multiply_Vector2D_ByFloat(
     Float=&FloatFromCurve(FloatCurve=DataInterface() {propertyValues {Curve {keys [
          {arriveTangent=4.1785, interpMode=RCIM_Cubic, leaveTangent=4.1785, tangentMode=RCTM_User, time=0, value=0}
          {arriveTangent=0.9995, interpMode=RCIM_Cubic, leaveTangent=0.9995, tangentMode=RCTM_User, time=0.2223, value=0.6825}
          {arriveTangent=-0.0405, interpMode=RCIM_Cubic, leaveTangent=-0.0405, tangentMode=RCTM_User, time=1, value=1}
         ]}}})
     Vector2D=&Vector2DFromFloat(Value=&RandomRangeFloat(Minimum=90, Maximum=120))
    )
    Write to Module Namespace=true
   ) {module=FN_Sprite}
   /SetVariables(Particles.SpriteColor=&MakeLinearColorFromVectorAndFloat(
     "Vector (RGB)"=&MakeVectorFromLinearColorRGB(Color=&ColorFromCurve(ColorCurve=DataInterface() {propertyValues {
         AlphaCurve {keys []}
         BlueCurve {keys [
           {arriveTangent=-0.238, interpMode=RCIM_Cubic, leaveTangent=-0.238, time=0, value=0.4871}
           {arriveTangent=-0.238, interpMode=RCIM_Cubic, leaveTangent=-0.238, time=1, value=0.2491}
          ]}
         GreenCurve {keys [
           {arriveTangent=-2.5716, interpMode=RCIM_Cubic, leaveTangent=-2.5716, time=0, value=2.8407}
           {arriveTangent=-2.5716, interpMode=RCIM_Cubic, leaveTangent=-2.5716, time=1, value=0.2691}
          ]}
         RedCurve {keys [
           {arriveTangent=-6.5, interpMode=RCIM_Cubic, leaveTangent=-6.5, time=0, value=7}
           {arriveTangent=-6.5, interpMode=RCIM_Cubic, leaveTangent=-6.5, time=1, value=0.5}
          ]}
        }}))
     "Float (Alpha)"=&FloatFromCurve(FloatCurve=DataInterface() {propertyValues {Curve {keys [<at 0 1>,<at 1 0>]}}})
    )) {isSetParameters=true, module=SetVariables_#2}
   &DynamicMaterialParameters(Write Parameter Index 0=false)
   &FN_Light_Attributes(
    Color=<rgba 2 1.3201 0.1458 1>
    Create Color Attribute=true
    Create Exponent Attribute=true
    Enable Particle Light Enable=true
    Exponent=3
    Light Enabled=true
    Scale Color=30
   )
   &FN_Light_Attributes(
    Color=<rgba 1 0.4871 0.0636 1>
    Create Color Attribute=true
    Create Radius Attribute=true
    Enable Particle Light Enable=true
    Light Enabled=true
    Radius=100
   ) {module=FN_Light_Attributes001}
   /SetVariables(Particles.LightRadius=50) {isSetParameters=true}
   /SetVariables(Particles.Light_Componet_intensity:NiagaraFloat=&Multiply_Float(
     A=130
     B=&FloatFromCurve(FloatCurve=DataInterface() {propertyValues {Curve {keys [
          {interpMode=RCIM_Cubic, tangentMode=RCTM_User, time=0, value=0}
          {interpMode=RCIM_Cubic, tangentMode=RCTM_User, time=0.5, value=1}
          {tangentMode=RCTM_User, time=1, value=0}
         ]}}})
    )) {isSetParameters=true}
  ]
  rendererClasses [&NiagaraMeshRendererProperties]
  renderers [{rendererClass=&NiagaraMeshRendererProperties, rendererIndex=0}]
 }
 FN_Weapons_Core_Flash_BETA002 {
  emitterUpdateScript [
   &EmitterState()
   &SpawnBurst_Instantaneous(Spawn Count=3, Use Spawn Probability=false)
  ]
  particleSpawnScript [
   &InitializeParticle(
    Color=<rgba 5.3739 3.8765 2.1281 1>
    Lifetime=&UniformRangedFloat(Minimum=0.1, Maximum=0.1)
    Mesh Scale=&UniformRangedVector(Minimum=<1.5 0.1 0.1>, Maximum=<2 0.25 0.25>)
    Module.Write Color=true
    Module.Write Lifetime=true
    Module.Write Mass=true
    Module.Write Material Random=true
    Module.Write Position=true
    Module.Write Scale=true
    Module.Write SpriteRotation=false
    Module.Write SpriteSize=true
    Position=&Add_Vector(A=&SimulationPosition(), B=<0 0 0>)
    Sprite Size=<100 100>
   )
   &CameraOffset(Camera Offset Amount=10)
   &InitialMeshOrientation(
    Mesh Orientation Mode=::NewEnumerator2
    Rotation=&RandomRangeVector(Minimum=<0 -0.2 -0.2>, Maximum=<1 0.2 0.2>)
    Use Rotation Vector=true
   )
  ]
  particleUpdateScript [
   &ParticleState()
   &Color(
    Color=&MakeLinearColorFromVectorAndFloat(
     "Vector (RGB)"=&MakeVectorFromLinearColorRGB(Color=&ColorFromCurve(ColorCurve=DataInterface() {propertyValues {
         AlphaCurve {keys []}
         BlueCurve {keys [
           {arriveTangent=-4.1816, interpMode=RCIM_Cubic, leaveTangent=-4.1816, time=0, value=4.1781}
           {arriveTangent=-4.1816, interpMode=RCIM_Cubic, leaveTangent=-4.1816, time=0.9977, value=0.0059}
          ]}
         GreenCurve {keys [
           {arriveTangent=-7.3723, interpMode=RCIM_Cubic, leaveTangent=-7.3723, time=0, value=7.3615}
           {arriveTangent=-7.3723, interpMode=RCIM_Cubic, leaveTangent=-7.3723, time=0.9977, value=0.0059}
          ]}
         RedCurve {keys [
           {arriveTangent=-14.4701, interpMode=RCIM_Cubic, leaveTangent=-14.4701, time=0, value=15}
           {arriveTangent=-14.4701, interpMode=RCIM_Cubic, leaveTangent=-14.4701, time=0.9977, value=0.5625}
          ]}
        }}))
     "Float (Alpha)"=&FloatFromCurve(FloatCurve=DataInterface() {propertyValues {Curve {keys [{interpMode=RCIM_Cubic, tangentMode=RCTM_User, time=0, value=1}, {interpMode=RCIM_Cubic, tangentMode=RCTM_User, time=1, value=0}]}}})
    )
    Scale Color=&VectorFromFloat(Value=&FloatFromCurve(FloatCurve=DataInterface() {propertyValues {Curve {keys [{interpMode=RCIM_Cubic, tangentMode=RCTM_User, time=0, value=1}, {interpMode=RCIM_Cubic, tangentMode=RCTM_User, time=1, value=0}]}}}))
   )
   &ScaleMeshSize(Scale Factor=&VectorFromCurve(VectorCurve=DataInterface() {propertyValues {XCurve {keys [<at 0 0.4>,<at 1 1.2>]}, YCurve {keys [<at 0 1>,<at 1 0>]}, ZCurve {keys [<at 0 1>,<at 1 0>]}}}))
   &DynamicMaterialParameters(Write Parameter Index 0=false)
  ]
  rendererClasses [&NiagaraMeshRendererProperties]
  renderers [{rendererClass=&NiagaraMeshRendererProperties, rendererIndex=0}]
 }
 FN_Weapons_Debris_Mesh_BETA {
  simTarget=GPUComputeSim
  emitterUpdateScript [
   &EmitterState()
   &SpawnBurst_Instantaneous(Spawn Count=&RandomRangeInt(Minimum=1, Maximum=2), Use Spawn Probability=false)
  ]
  particleSpawnScript [
   /SetVariables(Particles.Distribution:NiagaraFloat=&Add_Float(
     A=&UniformAOrBFloat(A=0.25, B=0.9, Distribution Weight=0.1)
     B=&RandomRangeFloat(Minimum=-0.1, Maximum=0.1)
    )) {isSetParameters=true}
   /Niagara/Modules/Spawn/Initialization/V2/InitializeParticle(
    Color=<rgba 1 0.9715 0.9375 1>
    Lifetime Max=1
    Lifetime Min=0.2
    Lifetime Mode=::NewEnumerator1
    Mesh Scale Max=<2 1 2>
    Mesh Scale Min=<0.9 0.7 1.2>
    Mesh Scale Mode=::NewEnumerator2
    UsePositionOffset=false
    Write Material Random=true
   ) {module=InitializeParticle001}
   &InitialMeshOrientation(Orientation Vector=&RandomVector(Vector Scale=1), Use Orientation Vector=true, Use Rotation Vector=false)
   &AddVelocityInCone(
    Cone Angle=40
    Module.Use Velocity Falloff On Cone Axis=false
    Velocity Strength=&RandomRangeFloat(Minimum=150, Maximum=425)
   )
   &ApplyInitialForces()
  ]
  particleUpdateScript [
   &ParticleState()
   &AccelerationForce(Acceleration=&MakeVector(
     X=0
     Y=0
     Z=/Game/Effects/Niagara/DynamicInputs/Math/FN_Fit01(Clamp:NiagaraBool=false, Value:NiagaraFloat=$Particles.Distribution, DestMin:NiagaraFloat=-980, DestMax:NiagaraFloat=-1200)
    )) {enabled=false}
   &GravityForce()
   &Drag(Drag=0.1, Use Linear Drag=true, Use Rotational Drag=false)
   &MeshRotationForce(Rotation=&RandomVector(Vector Scale=&RandomRangeFloat(Minimum=10, Maximum=20)))
   &SolveForcesAndVelocity(Clamp Velocity=false, Limit Acceleration=false)
   &SolveRotationalForcesAndVelocity()
   &FN_Mesh_Attributes(
    Create Scale Attribute=true
    Create Visibility Attribute=true
    Scale=&Multiply_Vector(
     A=$Particles.Initial.Scale
     B=&VectorFromFloat(Value=&FloatFromCurve(FloatCurve=DataInterface() {propertyValues {Curve {keys [
           <at -0.0032 0 RCIM_Cubic>
           {arriveTangent=-0.0978, interpMode=RCIM_Cubic, leaveTangent=-0.0978, tangentMode=RCTM_User, time=0.1249, value=0.9962}
           {arriveTangent=0.0267, interpMode=RCIM_Cubic, leaveTangent=0.0267, tangentMode=RCTM_User, time=0.7407, value=1.0021}
           <at 1 0 RCIM_Cubic>
          ]}}}))
    )
   ) {enabled=false}
  ]
  rendererClasses [&NiagaraMeshRendererProperties]
  renderers [{rendererClass=&NiagaraMeshRendererProperties, rendererIndex=0}]
 }
 FN_Weapons_Debris_Sprite_SubUV_BETA {
  simTarget=GPUComputeSim
  emitterUpdateScript [
   &EmitterState(
    Enable Distance Culling=true
    Max Distance Response=::KillAfterParticlesFinish
    Module.bUseMaxDistance=true
    Module.bUseMinDistance=false
    Scalability Mode=::NewEnumerator1
   )
   &SpawnBurst_Instantaneous(Spawn Count=&RandomRangeInt(Minimum=4, Maximum=14), Use Spawn Probability=false)
  ]
  particleSpawnScript [
   /Niagara/Modules/Spawn/Initialization/InitializeParticle(
    Color=<rgba 0.0125 0.0099 0.0068 1>
    Lifetime=&UniformRangedFloat(Minimum=0.2, Maximum=1)
    Module.Write Color=true
    Module.Write Lifetime=true
    Module.Write Mass=true
    Module.Write Material Random=true
    Module.Write Position=false
    Module.Write Scale=false
    Module.Write SpriteRotation=false
    Module.Write SpriteSize=true
    Position=/Niagara/Functions/Localspace/SimulationPosition()
    Sprite Size=&UniformRangedVector2D(Minimum=<2.1064 4.7118>, Maximum=<4.4317 8.5012>)
   )
   &AddVelocityInCone(
    Cone Angle=90
    Module.Use Velocity Falloff On Cone Axis=false
    Velocity Strength=&UniformRangedFloat(Minimum=200, Maximum=600)
   )
   /SetVariables(Particles.WoodVisibility:NiagaraInt32=&RandomRangeInt(Minimum=0, Maximum=1)) {isSetParameters=true}
   /SetVariables(Particles.WoodDebrisColor:LinearColor=&ColorFromCurve(
     ColorCurve=DataInterface() {propertyValues {
       AlphaCurve {keys [<at 0 1 RCIM_Cubic>]}
       BlueCurve {keys [<at 0 0.0247 RCIM_Cubic>,<at 1 0.0118 RCIM_Cubic>]}
       GreenCurve {keys [<at 0 0.0427 RCIM_Cubic>,<at 1 0.0235 RCIM_Cubic>]}
       RedCurve {keys [<at 0 0.0521 RCIM_Cubic>,<at 1 0.0392 RCIM_Cubic>]}
      }}
     CurveIndex=&RandomRangeFloat(Minimum=0, Maximum=1)
    )) {isSetParameters=true}
  ]
  particleUpdateScript [
   &ParticleState()
   &AccelerationForce(Acceleration=<0 0 -1200>)
   &Drag(Drag=0.1, Use Linear Drag=true, Use Rotational Drag=false)
   &SolveForcesAndVelocity(Clamp Velocity=false, Limit Acceleration=false)
   &ScaleSpriteSize(Scale Factor=&Vector2DFromFloat(Value=&FloatFromCurve(FloatCurve=DataInterface() {propertyValues {Curve {keys [<at 0.7424 1 RCIM_Cubic>,<at 1.0011 0.002>]}}})))
   /Niagara/Modules/Update/SubUV/SubUVAnimation(Random SubUV Lookup Index:NiagaraFloat=$Particles.Initial.MaterialRandom, SubUV Animation Mode=::NewEnumerator1)
   &Color(Color=$Particles.Initial.Color)
   &FN_Sprite_Attributes(
    Create Sprite Rotation Attribute=true
    Create SpriteSize Attribute=true
    SpriteRotation:NiagaraFloat=&RandomRangeFloat(Minimum=0, Maximum=360)
    SpriteSize=&RandomRangeVector2D(Minimum=<3 8>, Maximum=<6 25>)
    Write to Module Namespace=true
   )
  ]
  rendererClasses [&NiagaraSpriteRendererProperties]
  renderers [
   {rendererClass=&NiagaraSpriteRendererProperties}
   {rendererClass=&NiagaraSpriteRendererProperties, rendererIndex=1}
  ]
 }
 FN_Weapons_Smoke_BETA001 {
  emitterUpdateScript [
   &EmitterState(
    Enable Distance Culling=true
    Max Distance Response=::KillAfterParticlesFinish
    MaxDistance=7000
    Module.bUseMaxDistance=true
    Module.bUseMinDistance=false
    Scalability Mode=::NewEnumerator1
   )
   &SpawnBurst_Instantaneous(Spawn Count=&RandomRangeInt(Minimum=0, Maximum=1), Use Spawn Probability=false)
  ]
  particleSpawnScript [
   /Niagara/Modules/Spawn/Initialization/V2/InitializeParticle(
    Color=<rgba 0.3646 0.3646 0.3646 4>
    Lifetime Max=1
    Lifetime Min=0.15
    Lifetime Mode=::NewEnumerator1
    Sprite Rotation Mode=::NewEnumerator2
    Sprite Size Mode=::NewEnumerator1
    Uniform Sprite Size Max=80
    Uniform Sprite Size Min=50
    UsePositionOffset=false
    Write Material Random=true
   ) {module=InitializeParticle001}
   /SetVariables(Particles.InitSpriteRotRate=&UniformRangedFloat(Minimum=-3, Maximum=3)) {isSetParameters=true}
   &AddVelocityInCone(
    Cone Angle=10
    Module.Use Velocity Falloff On Cone Axis=false
    Velocity Strength=&UniformRangedFloat(Minimum=100, Maximum=550)
   )
  ]
  particleUpdateScript [
   &ParticleState()
   &SpriteRotationRate(Rotation Rate=&Multiply_Float(
     A=$Particles.InitSpriteRotRate
     B=&FloatFromCurve(FloatCurve=DataInterface() {propertyValues {Curve {keys [<at -0.0046 1.0113 RCIM_Cubic>,<at 1.0011 0.09 RCIM_Cubic>]}}})
    ))
   &Drag(Drag=6, Use Linear Drag=true, Use Rotational Drag=false)
   &SolveForcesAndVelocity(Clamp Velocity=false, Limit Acceleration=false)
   &ScaleSpriteSize(Scale Factor=&Vector2DFromFloat(Value=&FloatFromCurve(FloatCurve=DataInterface() {propertyValues {Curve {keys [
          {arriveTangent=2.2362, interpMode=RCIM_Cubic, leaveTangent=2.2362, tangentMode=RCTM_User, time=0, value=0.1}
          {arriveTangent=1.953, interpMode=RCIM_Cubic, leaveTangent=1.953, tangentMode=RCTM_User, time=0.2, value=0.9}
          {arriveTangent=0.0514, interpMode=RCIM_Cubic, leaveTangent=0.0514, tangentMode=RCTM_User, time=1.0008, value=1.6042}
         ]}}})))
   /Game/Effects/Niagara/Modules/Materials/MaterialSmoke(
    Erode:NiagaraFloat=&FloatFromCurve(FloatCurve=DataInterface() {propertyValues {
       bOptimizeLUT=false
       bUseLUT=false
       Curve {keys [
         {arriveTangent=-0.0652, interpMode=RCIM_Cubic, leaveTangent=-0.0652, tangentMode=RCTM_User, time=0.11, value=0.97}
         {arriveTangent=-5.7797, interpMode=RCIM_Cubic, leaveTangent=-5.7797, tangentMode=RCTM_User, time=0.39, value=0.66}
         {arriveTangent=-0.342, interpMode=RCIM_Cubic, leaveTangent=-0.342, tangentMode=RCTM_User, time=0.9954, value=-0.74}
        ]}
      }})
    Shading:NiagaraFloat=&FloatFromCurve(FloatCurve=DataInterface() {propertyValues {
       bOptimizeLUT=false
       bUseLUT=false
       Curve {keys [
         {arriveTangent=0.3846, interpMode=RCIM_Cubic, leaveTangent=0.3846, time=0.07, value=0.5}
         {arriveTangent=0.303, interpMode=RCIM_Cubic, leaveTangent=0.303, time=0.2, value=0.55}
         {interpMode=RCIM_Cubic, tangentMode=RCTM_User, time=0.4, value=0.6}
         {arriveTangent=-1.5, interpMode=RCIM_Cubic, leaveTangent=-1.5, time=0.8, value=0}
        ]}
      }})
   )
   &CameraOffset(Camera Offset Amount=20)
   &DynamicMaterialParameters(
    Index 0 Param 1=0.5
    Index 0 Param 2=&FloatFromCurve(FloatCurve=DataInterface() {propertyValues {Curve {keys [
         {arriveTangent=0.8564, interpMode=RCIM_Cubic, leaveTangent=0.8564, tangentMode=RCTM_User, time=0.0046, value=-0.9753}
         {arriveTangent=4.2876, interpMode=RCIM_Cubic, leaveTangent=5.5268, tangentMode=RCTM_User, time=1, value=1}
        ]}}})
    Index 0 Param 3=&FloatFromCurve(FloatCurve=DataInterface() {propertyValues {Curve {keys [
         {arriveTangent=-3.123, interpMode=RCIM_Cubic, leaveTangent=-3.123, leaveTangentWeight=0.3333, tangentMode=RCTM_User, time=0, value=0.5}
         {arriveTangentWeight=0.1167, interpMode=RCIM_Cubic, leaveTangentWeight=0.2167, tangentMode=RCTM_User, time=0.12, value=0}
        ]}}})
    Index 0 Param 4=&FloatFromCurve(FloatCurve=DataInterface() {propertyValues {Curve {keys [{arriveTangent=-3.2497, interpMode=RCIM_Cubic, leaveTangent=-3.2497, tangentMode=RCTM_User, time=0, value=1}, <at 1 0 RCIM_Cubic>]}}})
    Module.Param0WriteEnabled=true
    Module.Param1WriteEnabled=true
    Module.Param2WriteEnabled=true
    Module.Param3WriteEnabled=true
   )
   &ScaleColor(
    Scale Alpha=&FloatFromCurve(FloatCurve=DataInterface() {propertyValues {Curve {keys [<at -0.003 -0.006>,<at 0.1806 0.5257>,<at 0.7 0.35>,<at 1 0>]}}}, Scale Curve=0.2)
    Scale RGB=&VectorFromFloat(Value=&FloatFromCurve(FloatCurve=DataInterface() {propertyValues {Curve {keys [<at 0 0>,<at 0.1 1.5>,<at 1 0.5>]}}}))
    ScaleA=true
    ScaleRGB=true
   )
  ]
  rendererClasses [&NiagaraSpriteRendererProperties]
  renderers [
   {rendererClass=&NiagaraSpriteRendererProperties}
   {rendererClass=&NiagaraSpriteRendererProperties, rendererIndex=1}
  ]
 }
 FlashStreaks {
  emitterUpdateScript [
   &EmitterState(
    Enable Distance Culling=true
    Max Distance Response=::KillAfterParticlesFinish
    Module.bUseMaxDistance=true
    Module.bUseMinDistance=false
    Scalability Mode=::NewEnumerator1
   )
   &SpawnBurst_Instantaneous(Spawn Count=5, Use Spawn Probability=false)
  ]
  particleSpawnScript [
   /Niagara/Modules/Spawn/Initialization/InitializeParticle(
    Lifetime=&UniformRangedFloat(Minimum=0.12, Maximum=0.13)
    Mesh Scale=&UniformRangedVector(Minimum=<0.3 0.1 0.1>, Maximum=<0.6 0.2 0.2>)
    Module.Write Color=true
    Module.Write Lifetime=true
    Module.Write Mass=true
    Module.Write Material Random=true
    Module.Write Position=true
    Module.Write Scale=true
    Module.Write SpriteRotation=false
    Module.Write SpriteSize=true
    Position=&Add_Vector(A=/Niagara/DynamicInputs/Helpers/SimulationPosition(), B=<0 0 0>)
    Sprite Size=<100 100>
   )
   &FN_Fan(Cone Angle=180, Spin:NiagaraFloat=&UniformRangedFloat(Minimum=0, Maximum=1))
   &CameraOffset(Camera Offset Amount=5)
   /SetVariables(Particles.LightPositionNew:Vector3f=&Add_Vector(
     A=Linked() {linkedVariable {name=Particles.Position, type=NiagaraPosition}}
     B=&TransformPosition(Position=<25 0 0>, Source Space=::Local, Destination Space=::Simulation)
    )) {isSetParameters=true, module=SetVariables_#1}
  ]
  particleUpdateScript [
   &ParticleState()
   &Color(
    Color=<rgba 1 0.5325 0.0729 1>
    Scale Color=&VectorFromCurve(VectorCurve=DataInterface() {propertyValues {XCurve {keys [<at 0.7 12>,<at 1 1>]}, YCurve {keys [<at 0.7 12>,<at 1 1>]}, ZCurve {keys [<at 0.7 12>,<at 1 1>]}}})
   )
   &ScaleMeshSize(Scale Factor=&VectorFromFloat(Value=&FloatFromCurve(
      FloatCurve=DataInterface() {propertyValues {Curve {keys [{arriveTangent=2.2457, interpMode=RCIM_Cubic, leaveTangent=2.2457, tangentMode=RCTM_User, time=0, value=0}, <at 1 1>]}}}
      Scale Curve=0.42
     )))
   &FN_Sprite_Attributes(
    Create SpriteSize Attribute=true
    SpriteSize=&Vector2DFromFloat(Value=50)
    Write to Module Namespace=true
   ) {enabled=false, module=FN_Sprite}
   /SetVariables(Particles.SpriteColor=&MultiplyLinearColor(Linear Color=$Particles.Color, Scale Factor=<1 1 1 0.2>)) {isSetParameters=true, module=SetVariables_#2}
   &DynamicMaterialParameters(Write Parameter Index 0=false)
   &FN_Light_Attributes(
    Color=<rgba 1 0.6601 0.0729 1>
    Create Color Attribute=true
    Create Exponent Attribute=true
    Enable Particle Light Enable=true
    Exponent=4
    Light Enabled=true
    Scale Color=5
   ) {enabled=false}
   &FN_Light_Attributes(Color=<rgba 1 0.7728 0.204 1>, Create Color Attribute=true, Enable Particle Light Enable=true, Light Enabled=true) {enabled=false, module=FN_Light_Attributes001}
  ]
  rendererClasses [&NiagaraMeshRendererProperties]
  renderers [{rendererClass=&NiagaraMeshRendererProperties, rendererIndex=0}]
 }
}
summary {
 systemName=NS_Core_Shotgun_Impact_Wood
 emitters [
  {emitterName=FlashStreaks, rendererClasses [&NiagaraMeshRendererProperties]}
  {emitterName=FN_Weapons_Debris_Sprite_SubUV_BETA, simTarget=GPUComputeSim, rendererClasses [&NiagaraSpriteRendererProperties]}
  {emitterName=FN_Weapons_Debris_Mesh_BETA, simTarget=GPUComputeSim, rendererClasses [&NiagaraMeshRendererProperties]}
  {emitterName=FN_Weapons_Smoke_BETA001, rendererClasses [&NiagaraSpriteRendererProperties]}
  {emitterName=FN_Weapons_Core_Flash_BETA, rendererClasses [&NiagaraMeshRendererProperties]}
  {emitterName=FN_Weapons_Core_Flash_BETA002, rendererClasses [&NiagaraMeshRendererProperties]}
 ]
 userVariables []
}
